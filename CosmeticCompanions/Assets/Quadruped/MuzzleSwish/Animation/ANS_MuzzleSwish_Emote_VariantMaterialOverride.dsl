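; Anim notify state for the MuzzleSwish quadruped cosmetic companion that carries its variant look onto an emote mesh's materials.
; On notify begin, it casts the mesh's attach parent to a skeletal mesh and that mesh's anim instance to MuzzleSwish_Base_AnimBP_Child.
; It reads the variant's body and accent IDs there and writes them as BodyID and AccentID material scalars, without checking the success flag.
; Failed casts change nothing.
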(fn Received_NotifyBegin
  (params (MeshComp :type Skeletal%20Mesh%20Component%20Object%20Reference) (Animation :type Anim%20Sequence%20Base%20Object%20Reference) (TotalDuration :type Float%20%28single-precision%29) (EventReference :type Anim%20Notify%20Event%20Reference%20Structure%20%28by%20ref%29))
  (bind _asskeletal_mesh_component (Utilities|Casting|CastToSkeletalMeshComponent (Transformation|GetAttachParent MeshComp))
    (:then
      (bind _asmuzzle_swish_base_anim_bp_child (Utilities|Casting|CastToMuzzleSwish_Base_AnimBP_Child (Components|SkeletalMesh|GetAnimInstance _asskeletal_mesh_component))
        (:then
          (bind (_success _bodyid _accentid) (Class|MuzzleSwishBaseAnimBPChild|GetMuzzleSwishVariantInfo _asmuzzle_swish_base_anim_bp_child))
          (Rendering|Material|SetScalarParameterValueonMaterials (Variables|GetMeshComp) "BodyID" (Math|Conversions|ToFloat(Integer) _bodyid))
          (Rendering|Material|SetScalarParameterValueonMaterials (Variables|GetMeshComp) "AccentID" (Math|Conversions|ToFloat(Integer) _accentid))
          (return false))
        (:CastFailed
          (return false))))
    (:CastFailed
      (return false))))
