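# Child Animation Blueprint for the MuzzleSwish quadruped cosmetic companion that reports which body and accent material variants are selected.
# GetMuzzleSwishVariantInfo switches on the gameplay tags from two immutable MaterialParamSets channels, mapping Mat0 through Mat2 to integer Body and Accent IDs.
# It returns both IDs with a success flag, or failure with zeroed IDs when either tag is unmatched.
# The update event only forwards to the parent.
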AnimBlueprint /CosmeticCompanions/Assets/Quadruped/MuzzleSwish/Blueprint/MuzzleSwish_Base_AnimBP_Child
target_skeleton=/CosmeticCompanions/Assets/Quadruped/MuzzleSwish/Meshes/MuzzleSwish_Quadruped_Skeleton
graphs [
 EventGraph {nodes [
   n1 = Event BlueprintUpdateAnimation()
   > Parent: BlueprintUpdateAnimation!(execute, DeltaTimeX:float=@n1.DeltaTimeX) -> (then) {is=title K2Node_CallParentFunction}
   TryGetPawnOwner()
  ]}
 GetMuzzleSwishVariantInfo {nodes [
   GetMuzzleSwishVariantInfo() {is=title K2Node_FunctionEntry}
   > n1 = Switch on Gameplay Tag!(
    execute:Exec
    Selection:Gameplay Tag Structure=GetVariantTagForChannel!(
     self:Fort Player Companion Anim Instance Object Reference
     ChannelTag:Gameplay Tag Structure=(TagName="Cosmetics.Variant.Channel.Immutable.MaterialParamSets")
    ) -> (ReturnValue:"Gameplay Tag Structure (by ref)")
   ) -> (Default:Exec, Cosmetics.Variant.Property.Mat0:Exec, Cosmetics.Variant.Property.Mat1:Exec, Cosmetics.Variant.Property.Mat2:Exec) {
    is=title GameplayTagsK2Node_SwitchGameplayTag
    pinTags [{tagName=Cosmetics.Variant.Property.Mat0}, {tagName=Cosmetics.Variant.Property.Mat1}, {tagName=Cosmetics.Variant.Property.Mat2}]
   }
   n2 = Set Body(execute=@n1.Cosmetics.Variant.Property.Mat0, Body:int=0) -> (Output_Get:int=0)
   n3 = Switch on Gameplay Tag(
    execute=[@n2.then, @n7.then, @n8.then]
    Selection=GetVariantTagForChannel(ChannelTag=(TagName="Cosmetics.Variant.Channel.Immutable.MaterialParamSets2"))
   ) {pinTags [{tagName=Cosmetics.Variant.Property.Mat0}, {tagName=Cosmetics.Variant.Property.Mat1}, {tagName=Cosmetics.Variant.Property.Mat2}]}
   n4 = Set Accent(execute=@n3.Cosmetics.Variant.Property.Mat0, Accent:int=0) -> (Output_Get:int=0)
   Return Node!(
    execute=[@n6.then, @n4.then, @n5.then]
    Success:bool=true
    BodyID:int=Get Body() -> (Body:int=0)
    AccentID:int=Get Accent() -> (Accent:int=0)
   )
   n5 = Set Accent(execute=@n3.Cosmetics.Variant.Property.Mat1, Accent=1) -> (Output_Get=0)
   n6 = Set Accent(execute=@n3.Cosmetics.Variant.Property.Mat2, Accent=2) -> (Output_Get=0)
   Return Node(execute=@n3.Default, Success=false, BodyID=0, AccentID=0)
   n7 = Set Body(execute=@n1.Cosmetics.Variant.Property.Mat1, Body=1) -> (Output_Get=0)
   n8 = Set Body(execute=@n1.Cosmetics.Variant.Property.Mat2, Body=2) -> (Output_Get=0)
   Return Node(execute=@n1.Default, Success=false, BodyID=0, AccentID=0)
  ]}
]
variables []
