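# Animation Blueprint for the MuzzleSwish quadruped cosmetic companion, driving its pose from another mesh.
# Each update sets useCopyPose when the owner is not a FortCompanionMannequin and bUseReverseRetarget is true.
# That flag blends between copying the attached parent's pose with curves and an IK-retargeted pose with the EmissiveBrightness curve set to 1.
# A second retarget node using the emote retargeter is left unconnected.
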AnimBlueprint /CosmeticCompanions/Assets/Quadruped/MuzzleSwish/Meshes/MuzzleSwish_Quadruped_AnimBP
target_skeleton=&/MuzzleSwish_Quadruped_Skeleton
graphs [
 AnimGraph {nodes [
   Retarget Pose From Mesh!() -> (Pose:Pose Link Structure) {
    is=title AnimGraphNode_RetargetPoseFromMesh
    bSuppressWarnings=true
    iKRetargeterAsset=&/RTG_MuzzleSwish_Emote
   }
   n1 = Retarget Pose From Mesh() {bSuppressWarnings=true, iKRetargeterAsset=&/RTG_MuzzleSwish}
   Output Pose(Result=Blend Poses by bool(
     BlendPose_0=Copy Pose From Mesh() {bCopyCurves=true, bUseAttachedParent=true}
     BlendPose_1=Modify Curve(SourcePose=@n1, CurveValues_0=1, Alpha=1) {curveNames [EmissiveBrightness], curveValues [0]}
     BlendTime_0=0.2
     BlendTime_1=0.2
     bActiveValue=Get useCopyPose() -> (useCopyPose:bool)
    ))
  ]}
 EventGraph {nodes [
   Event BlueprintUpdateAnimation()
   > n1 = Set isFortCompanionMannequin(isFortCompanionMannequin:bool=IsA!(Object:Object Reference=GetOwningActor(), ObjectClass:Object Class Reference=/Script/CosmeticCompanionsRuntime.FortCompanionMannequin) -> (ReturnValue:bool)) -> (Output_Get:bool)
   > Set useCopyPose(useCopyPose:bool=AND Boolean(A=NOT Boolean(A=@n1), B=Get bUseReverseRetarget() -> (bUseReverseRetarget:bool))) -> (Output_Get:bool)
  ]}
]
variables [
 isFortCompanionMannequin {type=boolean}
 useCopyPose {type=boolean}
]
