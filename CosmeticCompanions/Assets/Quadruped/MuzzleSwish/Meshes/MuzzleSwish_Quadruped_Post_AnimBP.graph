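# Post-process AnimBP for the MuzzleSwish quadruped companion, adding facial poses and secondary motion for its bow bones.
# The input pose drives a facial PoseAsset, then a DisableTongue blend mask weighted by DisableTongueWeight.
# A weighted constraint ties the bow master bone to spine_01 and spine_02.
# Spring-based AnimDynamics then simulate the right and left bow bones bone-relative to root, with DynamicsAlpha inverted by scale and bias.
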AnimBlueprint /CosmeticCompanions/Assets/Quadruped/MuzzleSwish/Meshes/MuzzleSwish_Quadruped_Post_AnimBP
target_skeleton=&/MuzzleSwish_Quadruped_Skeleton
graphs [
 AnimGraph {nodes [
   n1 = Layered blend per bone(
    BasePose=MuzzleSwish_Quadruped_Head_Facial_Poses_PoseAsset!(SourcePose:Pose Link Structure=Input Pose()) -> (Pose:Pose Link Structure) {
     is=title AnimGraphNode_PoseBlendNode
     poseAsset=&/MuzzleSwish_Quadruped_Head_Facial_Poses_PoseAsset
    }
    BlendWeights_0=Get DisableTongueWeight() -> (DisableTongueWeight:double)
   ) {
    blendMode=BlendMask
    curveBlendOption=DoNotOverride
    blendMasks [&MuzzleSwish_Quadruped_Skeleton:DisableTongue]
    layerSetup []
   }
   n2 = "Constraint - dyn_bow_mstr (PP)"!(ComponentPose=Local To Component(LocalPose=@n1), Alpha=1, ConstraintWeights_0=0.4, ConstraintWeights_1:float=0.6) -> (Pose) {
    boneToModify {boneName=dyn_bow_mstr}
    constraintSetup [{targetBone {boneName=spine_01}}, {targetBone {boneName=spine_02}}]
    constraintWeights [1,1]
   }
   n3 = Get DynamicsAlpha() -> (DynamicsAlpha:float)
   n4 = Anim Dynamics - Bone: dyn_bow_r(ComponentPose=@n2, Alpha=@n3, ExternalForce="0,0,0") {
    angularDampingOverride=0.9999
    angularSpringConstant=10000
    bAngularSpring=true
    bOverrideAngularDamping=true
    bUseGravityOverride=true
    gravityScale=0.35
    simulationSpace=BoneRelative
    alphaBoolBlend {blendOption=Linear}
    alphaScaleBiasClamp {bias=1, scale=-1}
    boundBone {boneName=dyn_bow_r}
    componentAppliedLinearAccClamp=<4000 4000 4000>
    componentLinearAccScale=<1 1 1>
    componentLinearVelScale=<1 1 1>
    gravityOverride {z=-980}
    physicsBodyDefinitions [{
      boundBone {boneName=dyn_bow_r}
      boxExtents=<10 10 10>
      constraintSetup {angularLimitsMax=<5 10 15>, angularLimitsMin=<-5 -10 -15>, angularTarget=<1 0 0>}
      localJointOffset=<2 0 0>
     }]
    relativeSpaceBone {boneName=root}
   }
   n5 = Anim Dynamics - Bone: dyn_bow_l(ComponentPose=@n4, Alpha=@n3, ExternalForce="0,0,0") {
    angularDampingOverride=0.9999
    angularSpringConstant=10000
    bAngularSpring=true
    bOverrideAngularDamping=true
    bUseGravityOverride=true
    gravityScale=0.35
    simulationSpace=BoneRelative
    alphaBoolBlend {blendOption=Linear}
    alphaScaleBiasClamp {bias=1, scale=-1}
    boundBone {boneName=dyn_bow_l}
    componentAppliedLinearAccClamp=<4000 4000 4000>
    componentLinearAccScale=<1 1 1>
    componentLinearVelScale=<1 1 1>
    gravityOverride {z=-980}
    physicsBodyDefinitions [{
      boundBone {boneName=dyn_bow_l}
      boxExtents=<12 12 12>
      constraintSetup {angularLimitsMax=<10 10 10>, angularLimitsMin=<-10 -10 -10>, angularTarget=<1 0 0>}
      localJointOffset=<2 0 0>
     }]
    relativeSpaceBone {boneName=root}
   }
   Output Pose(Result=Component To Local(ComponentPose=@n5))
  ]}
 EventGraph {links [], nodes []}
]
variables [
 DisableTongueWeight {type=number}
 PoseAsset {type=PoseAsset, default=&MuzzleSwish_Quadruped_Head_Facial_Poses_PoseAsset}
]
